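; Quest actor for collecting battle plans in Part 1 of the Season 22 Battle Pass narrative quests.
; ConsolidatedQuestComponent ready and not-ready events show or hide the pre-interaction quest mesh for that player's controller only.
; The visibility change runs only on non-authority clients and is skipped when the design flag for testing visuals without a quest is set.
; All other logic is left to the parent class.
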(event EventBeginPlay
  (|CallParentFunction) ; This node is disabled and will not be called.\nDrag off pins to build functionality.
)

(event Collision|EventActorBeginOverlap
  (params (OtherActor :type Actor%20Object%20Reference))
  (|CallParentFunction OtherActor) ; This node is disabled and will not be called.\nDrag off pins to build functionality.
)

(event EventTick
  (params (DeltaSeconds :type Float%20%28single-precision%29))
  (|CallParentFunction DeltaSeconds) ; This node is disabled and will not be called.\nDrag off pins to build functionality.
)

(event OnQuestReady(ConsolidatedQuestComponent)
  (params (PlayerController :type Fort%20Player%20Controller%20Object%20Reference))
  (Networking|SwitchHasAuthority
    (:Remote
      (if (not (Variables|Design|QuestSetup|GetbTestVisualsNoQuestRequired))
        (PlayerController|SetCompHiddenControllerOnly PlayerController (Variables|Components|GetSM_Quest_PreInteraction) false true)))))

(event OnQuestNotReady(ConsolidatedQuestComponent)
  (params (PlayerController :type Fort%20Player%20Controller%20Object%20Reference))
  (Networking|SwitchHasAuthority
    (:Remote
      (if (not (Variables|Design|QuestSetup|GetbTestVisualsNoQuestRequired))
        (PlayerController|SetCompHiddenControllerOnly PlayerController (Variables|Components|GetSM_Quest_PreInteraction) true true)))))

(fn ConstructionScript
  (|CallParentFunction))
