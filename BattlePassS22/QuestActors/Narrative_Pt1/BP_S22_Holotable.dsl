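; Season 22 narrative quest holotable that, on local interact, starts a client-side looping, spinning house hologram display.
; Looping swap and spin timelines share one play rate; each swap event picks one of three meshes.
; The index reset precedes its increment, so later loops skip the first mesh.
; Swaps spawn a Niagara transition, delay briefly, then scale the mesh up while spin rate varies inversely with scale.
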(event EventBeginPlay
  (|CallParentFunction))

(event EventEndPlay
  (params (EndPlayReason :type EEndPlayReason%20Enum :default "Destroyed"))
  (|CallParentFunction EndPlayReason))

(event Custom|HologramStateLoop
  ;[0 0 0] Client only - swap between the different house holograms for a set duration
  (CallFunction|SetupTimelines (Utilities|Array|MakeArray (Variables|BP_S22_Holotable|GetHologramSwapTimeline) (Variables|BP_S22_Holotable|GetSpinTimeline)))
  (Utilities|FlowControl|Sequence
    (:then_0
      (Variables|Default|SetHolotableLoopIteration 0)
      (Timeline|HologramSwapTimeline PlayFromStart
        (:SwapEvent
          (switch int (Variables|Default|GetHolotableLoopIteration)
            (:0
              (CallFunction|ChangeHologramMesh (Variables|Holotable|GetHologramHouseMesh_01) (Variables|Default|GetHolotableDefaultTransform)))
            (:1
              (CallFunction|ChangeHologramMesh (Variables|Holotable|GetHologramHouseMesh_02) (Variables|Default|GetHolotableDefaultTransform)))
            (:2
              (CallFunction|ChangeHologramMesh (Variables|Holotable|GetHologramHouseMesh_03) (Variables|Default|GetHolotableFinalTransform))
              (Variables|Default|SetHolotableLoopIteration 0)))
          (Variables|Default|SetHolotableLoopIteration (+ (Variables|Default|GetHolotableLoopIteration) 1)))))
    (:then_1
      (bind (_direction _spin) (Timeline|SpinTimeline PlayFromStart)
        (:Update
          (Transformation|SetRelativeRotation (Variables|Default|GetSM_Hologram) (Math|Rotator|MakeRotator 0.0 0.0 _spin) false false))))))

(event Custom|SetupTimelines
  (params (Timelines :type Array%20of%20Timeline%20Component%20Object%20References))
  ;[0 0 0] Client only - swap between the different house holograms for a set duration
  ;;[0.2396 0.2396 0.2396] For each timeline passed in, set them to loop and adjust play rate based on the float var
  (for _array_element Timelines
    (Components|Timeline|SetPlayRate _array_element (Variables|Default|GetLoopPlayRate))
    (Components|Timeline|SetLooping _array_element true)))

(event Custom|ChangeHologramMesh
  (params (HologramMesh :type Static%20Mesh%20Object%20Reference) (Transform :type Transform))
  (bind _sm_hologram (Variables|Default|GetSM_Hologram))
  ;[0 0 0] Client only - swap between the different house holograms for a set duration
  ;;[0.1927 0.1927 0.1927] Handles spawning the FX and the delay between the FX and transition
  (Niagara|SpawnSystemAttached (Variables|Holotable|GetNS_Hologram_TransitionEffect) (Variables|Default|GetNS_Hologram_Transition) "None" "0, 0, 0" "0, 0, 0" "KeepRelativeOffset" true)
  (Utilities|FlowControl|Delay 0.25)
  (Components|StaticMesh|SetStaticMesh _sm_hologram HologramMesh)
  (Transformation|SetRelativeTransform _sm_hologram (Math|Transform|MakeTransform (.location Transform) (.rotation Transform) (Variables|Default|GetStartingMeshScale)) false false)
  (bind (_direction _scaleuptimeline) (Timeline|HouseScaleupTimeline PlayFromStart)
    (:Update
      (Transformation|SetRelativeScale3D (Variables|Default|GetSM_Hologram) (* (Math|Vector|MakeVector _scaleuptimeline _scaleuptimeline _scaleuptimeline) (.scale Transform)))
      (Components|Timeline|SetPlayRate (Variables|BP_S22_Holotable|GetSpinTimeline) (/ (Variables|Default|GetLoopPlayRate) (/ _scaleuptimeline 1e+01))))
    (:Finished
      (Components|Timeline|SetPlayRate (Variables|BP_S22_Holotable|GetSpinTimeline) (Variables|Default|GetLoopPlayRate)))))

(fn BlueprintOnLocalInteract
  (params (InteractingPawn :type Fort%20Player%20Pawn%20Object%20Reference) (InteractionBeingAttempted :type EInteractionBeingAttempted%20Enum :default "FirstInteraction"))
  (bind _returnvalue (|CallParentFunction InteractingPawn InteractionBeingAttempted))
  (VisualState|Meshes|SetMeshHidden (Variables|Default|GetSM_Hologram) (Fortnite|GetFortPlayerControllerfromActor InteractingPawn))
  (CallFunction|HologramStateLoop)
  (return _returnvalue))

(fn ConstructionScript
  (bind _sm_hologram (Variables|Default|GetSM_Hologram))
  (|CallParentFunction)
  (Utilities|FlowControl|Sequence
    (:then_0
      (bind _output_get (Variables|Default|SetHolotableDefaultTransform (Transformation|GetRelativeTransform _sm_hologram)))
      (Transformation|SetRelativeTransform _sm_hologram _output_get false false)
      (Rendering|Material|SetMaterial (Variables|Default|GetSM_Hologram) 0 (Variables|Holotable|GetHologramMaterialOverride)))
    (:then_1
      (Niagara|SetNiagaraSystemAsset (Variables|Default|GetNS_Hologram_Transition) (Variables|Holotable|GetNS_Hologram_TransitionEffect) true))))
