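; Season 22 narrative quest terminal that changes its ambient audio as a player converses with it.
; ConversationMessage extends its parent and passes the message number through SwapAmbient to SwapAudioFunction.
; That runs only without authority, crossfading terminal audio components by message number if a local player's pawn carries the Conversation.InteractingWithNPC tag, otherwise fading out the first track.
; The NoVFX&Audio override also fades out that first track after calling its parent.
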(event Custom|Swap%20Ambient
  (params (Message%20Number :type Integer))
  (CallFunction|SwapAudioFunction Message%20Number))

(fn ConversationMessage
  (params (MessageNumber :type Integer))
  (|CallParentFunction MessageNumber)
  (CallFunction|SwapAmbient MessageNumber))

(fn NoVFX&Audio
  (params (PlayerController :type Fort%20Player%20Controller%20Object%20Reference))
  (|CallParentFunction PlayerController)
  (Audio|Components|Audio|FadeOut (Variables|Default|GetTerminalAudio_0) (Variables|Design|CosmeticEvent|GetCosmeticAudioFadeOut) 0.0 "Linear" :!class /Script/Engine.AudioComponent))

(fn SwapAudioFunction
  (params (Selection :type Integer))
  (bind _cosmeticaudiofadeout (Variables|Design|CosmeticEvent|GetCosmeticAudioFadeOut))
  (Networking|SwitchHasAuthority
    (:Remote
      (bind _array_element (Utilities|Array|ForEachLoopwithBreak (Fortnite|GetLocalFortPlayerControllers true))
        (:LoopBody
          (bind _returnvalue (GameplayTags|HasMatchingGameplayTag(Message) (Game|Player|GetPlayerPawn _array_element) "(TagName=\"Conversation.InteractingWithNPC\")"))
          (if _returnvalue
            (LocalVariables|SetShouldSwap true)))
        (:Completed
          _))
      (if (LocalVariables|GetShouldSwap)
        (switch int Selection
          (:Default
            _)
          (:0
            (Audio|Components|Audio|FadeIn (Variables|Default|GetTerminalAudio_0) _cosmeticaudiofadeout 1.0 0.0 "Linear" :!class /Script/Engine.AudioComponent))
          (:1
            (Audio|Components|Audio|FadeIn (Variables|Default|GetTerminalAudio_1) _cosmeticaudiofadeout 1.0 0.0 "Linear" :!class /Script/Engine.AudioComponent)
            (Audio|Components|Audio|FadeOut (Variables|Default|GetTerminalAudio_0) _cosmeticaudiofadeout 0.0 "Linear" :!class /Script/Engine.AudioComponent))
          (:2
            (Audio|Components|Audio|FadeIn (Variables|Default|GetTerminalAudio_2) _cosmeticaudiofadeout 1.0 0.0 "Linear" :!class /Script/Engine.AudioComponent)
            (Audio|Components|Audio|FadeOut (Variables|Default|GetTerminalAudio_1) _cosmeticaudiofadeout 0.0 "Linear" :!class /Script/Engine.AudioComponent))
          (:3
            (Audio|Components|Audio|FadeIn (Variables|Default|GetTerminalAudio_3) _cosmeticaudiofadeout 1.0 0.0 "Linear" :!class /Script/Engine.AudioComponent)
            (Audio|Components|Audio|FadeOut (Variables|Default|GetTerminalAudio_0) _cosmeticaudiofadeout 0.0 "Linear" :!class /Script/Engine.AudioComponent))
          (:4
            _))
        (Audio|Components|Audio|FadeOut (Variables|Default|GetTerminalAudio_0) _cosmeticaudiofadeout 0.0 "Linear" :!class /Script/Engine.AudioComponent)))))

(fn ConstructionScript
  (|CallParentFunction))
