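; Prototype test-chamber button quest prop for the Battle Pass S22 narrative quest, which turns red when triggered by overlap.
; On Box overlap, it finds the player's Athena quest manager and checks whether the quest from its ConsolidatedQuestComponent is in progress.
; If so, or if bTestVisualsNoQuestRequired is set, it applies a red-light material to the pre-interaction mesh.
; Nothing here reverts the material or advances the quest.
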(event EventBeginPlay
  (|CallParentFunction) ; This node is disabled and will not be called.\nDrag off pins to build functionality.
)

(event Collision|EventActorBeginOverlap
  (params (OtherActor :type Actor%20Object%20Reference))
  (|CallParentFunction OtherActor) ; This node is disabled and will not be called.\nDrag off pins to build functionality.
)

(event EventTick
  (params (DeltaSeconds :type Float%20%28single-precision%29))
  (|CallParentFunction DeltaSeconds) ; This node is disabled and will not be called.\nDrag off pins to build functionality.
)

(event OnComponentBeginOverlap(Box)
  (params (OverlappedComponent :type Primitive%20Component%20Object%20Reference) (OtherActor :type Actor%20Object%20Reference) (OtherComp :type Primitive%20Component%20Object%20Reference) (OtherBodyIndex :type Integer) (bFromSweep :type Boolean) (SweepResult :type Hit%20Result%20Structure%20%28by%20ref%29))
  (bind _asfort_player_pawn (Utilities|Casting|CastToFortPlayerPawn OtherActor))
  (if (or (Quest|IsQuestinProgress (Quests|GetQuestManager (Fortnite|GetFortPlayerControllerfromActor _asfort_player_pawn) "Athena") (Class|ConsolidatedQuestComponent|GetQuestItemDefinition (Variables|BuildingGameplayActorPropQuest|GetConsolidatedQuestComponent))) (Variables|Design|QuestSetup|GetbTestVisualsNoQuestRequired))
    (Rendering|Material|SetMaterial (Variables|Components|GetSM_Quest_PreInteraction) 0 "/BattlePassS22/QuestActors/Narrative_Pt1/TestChamber_Game/PrototypeMats/Test_Mat_RedLight.Test_Mat_RedLight")))

(fn ConstructionScript
  (|CallParentFunction))
